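# Physics asset for the DroveRay Dip medium male player body, set up for dynamic robe motion against the legs and ground.
# Seven five-bone robe chains around back, sides and front use capsules, linked only along each chain by twist-locked constraints.
# Top links swing on one axis; that limit generally widens downward, except the looser upper front-left chain.
# Thigh, calf and foot capsules provide leg shapes, and a huge sphere below the root approximates the floor.
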PhysicsAsset /BRCosmetics/Characters/Player/Male/Medium/Bodies/M_MED_DroveRay_Dip/Meshes/M_MED_DroveRay_Dip_Physics
bodies {
 calf_l [
  {shapeType=Capsule, length=41.8392, radius=6.5691, center=<-21.3994 -1.1783 0.1724>, rotation=<rot 0.0127 -91.5763 -89.5244>}
  {
   shapeType=Capsule
   length=28.2845
   radius=7.0562
   shapeName=calf_l_capsule
   center=<-15.2933 -0.7091 0.7805>
   rotation=<rot -89.7294 1.6405 -179.9988>
  }
 ]
 calf_r [
  {shapeType=Capsule, length=41.8392, radius=6.5691, center=<21.3994 1.1783 -0.1724>, rotation=<rot 0.0127 -91.5763 90.4756>}
  {
   shapeType=Capsule
   length=28.2845
   radius=7.0562
   shapeName=calf_r_capsule
   center=<15.2933 0.7091 -0.7805>
   rotation=<rot -89.7294 1.6405 0.0012>
  }
 ]
 dyn_robeBack_01 [{
   shapeType=Capsule
   length=33.1081
   radius=14.1636
   shapeName=dyn_robeSideFront_05_l_capsule
   center=<3.2636 -5.0836 0>
   rotation=<rot -8.8651 -89.5938 -2.6352>
  }]
 dyn_robeBack_02 [{
   shapeType=Capsule
   length=33.1081
   radius=14.1636
   shapeName=dyn_robeSideFront_05_l_capsule
   center=<3.2636 -5.0836 0>
   rotation=<rot -8.8651 -89.5938 -2.6352>
  }]
 dyn_robeBack_03 [{
   shapeType=Capsule
   length=33.1081
   radius=14.1636
   shapeName=dyn_robeSideFront_05_l_capsule
   center=<3.2636 -5.0836 0>
   rotation=<rot -8.8651 -89.5938 -2.6352>
  }]
 dyn_robeBack_04 [{
   shapeType=Capsule
   length=33.1081
   radius=14.1636
   shapeName=dyn_robeSideFront_05_l_capsule
   center=<3.2636 -5.0836 0>
   rotation=<rot -8.8651 -89.5938 -2.6352>
  }]
 dyn_robeBack_05 [{
   shapeType=Capsule
   length=33.1081
   radius=14.1636
   shapeName=dyn_robeSideFront_05_l_capsule
   center=<3.2636 -5.0836 0>
   rotation=<rot -7.814 -89.6425 -2.6281>
  }]
 dyn_robeSideBack_01_l [{
   shapeType=Capsule
   length=33.1081
   radius=14.1636
   shapeName=dyn_robeSideFront_05_l_capsule
   center=<3.2636 -5.0836 0>
   rotation=<rot -8.8651 -89.5938 -2.6352>
  }]
 dyn_robeSideBack_01_r [{
   shapeType=Capsule
   length=33.1081
   radius=14.1636
   shapeName=dyn_robeSideFront_05_l_capsule
   center=<3.2636 -5.0836 0>
   rotation=<rot -8.8651 -89.5938 -2.6352>
  }]
 dyn_robeSideBack_02_l [{
   shapeType=Capsule
   length=33.1081
   radius=14.1636
   shapeName=dyn_robeSideFront_05_l_capsule
   center=<3.2636 -5.0836 0>
   rotation=<rot -8.8651 -89.5938 -2.6352>
  }]
 dyn_robeSideBack_02_r [{
   shapeType=Capsule
   length=33.1081
   radius=14.1636
   shapeName=dyn_robeSideFront_05_l_capsule
   center=<3.2636 -5.0836 0>
   rotation=<rot -8.8651 -89.5938 -2.6352>
  }]
 dyn_robeSideBack_03_l [{
   shapeType=Capsule
   length=33.1081
   radius=14.1636
   shapeName=dyn_robeSideFront_05_l_capsule
   center=<3.2672 -5.591 -0.0792>
   rotation=<rot -8.8651 -89.5938 -2.6352>
  }]
 dyn_robeSideBack_03_r [{
   shapeType=Capsule
   length=33.1081
   radius=14.1636
   shapeName=dyn_robeSideFront_05_l_capsule
   center=<3.274 -6.5536 -0.2293>
   rotation=<rot -8.8651 -89.5938 -2.6352>
  }]
 dyn_robeSideBack_04_l [{
   shapeType=Capsule
   length=33.1081
   radius=14.1636
   shapeName=dyn_robeSideFront_05_l_capsule
   center=<3.2573 -4.606 0.139>
   rotation=<rot -16.2236 -89.2419 -2.7118>
  }]
 dyn_robeSideBack_04_r [{
   shapeType=Capsule
   length=33.1081
   radius=14.1636
   shapeName=dyn_robeSideFront_05_l_capsule
   center=<3.2703 -6.3522 -0.148>
   rotation=<rot -0.8336 -89.9622 -2.6039>
  }]
 dyn_robeSideBack_05_l [{
   shapeType=Capsule
   length=33.1081
   radius=14.1636
   shapeName=dyn_robeSideFront_05_l_capsule
   center=<3.2636 -5.0836 0>
   rotation=<rot -22.5815 -88.9164 -2.8202>
  }]
 dyn_robeSideBack_05_r [{
   shapeType=Capsule
   length=33.1081
   radius=14.1636
   shapeName=dyn_robeSideFront_05_l_capsule
   center=<3.2676 -5.6469 -0.0879>
   rotation=<rot 8.7153 -90.3994 -2.634>
  }]
 dyn_robeSideFront_01_l [{
   shapeType=Capsule
   length=33.1081
   radius=14.1636
   shapeName=dyn_robeSideFront_05_l_capsule
   center=<3.2636 -5.0836 0>
   rotation=<rot -8.8651 -89.5938 -2.6352>
  }]
 dyn_robeSideFront_01_r [{
   shapeType=Capsule
   length=33.1081
   radius=14.1636
   shapeName=dyn_robeSideFront_05_l_capsule
   center=<3.2636 -5.0836 0>
   rotation=<rot -8.8651 -89.5938 -2.6352>
  }]
 dyn_robeSideFront_02_l [{
   shapeType=Capsule
   length=33.1081
   radius=14.1636
   shapeName=dyn_robeSideFront_05_l_capsule
   center=<3.2636 -5.0836 0>
   rotation=<rot -8.8651 -89.5938 -2.6352>
  }]
 dyn_robeSideFront_02_r [{
   shapeType=Capsule
   length=33.1081
   radius=14.1636
   shapeName=dyn_robeSideFront_05_l_capsule
   center=<3.2636 -5.0836 0>
   rotation=<rot -8.8651 -89.5938 -2.6352>
  }]
 dyn_robeSideFront_03_l [{
   shapeType=Capsule
   length=33.1081
   radius=14.1636
   shapeName=dyn_robeSideFront_05_l_capsule
   center=<3.2904 -8.8547 -0.5882>
   rotation=<rot -8.8651 -89.5938 -2.6352>
  }]
 dyn_robeSideFront_03_r [{
   shapeType=Capsule
   length=33.1081
   radius=14.1636
   shapeName=dyn_robeSideFront_05_l_capsule
   center=<3.2904 -8.8612 -0.5892>
   rotation=<rot -8.8651 -89.5938 -2.6352>
  }]
 dyn_robeSideFront_04_l [{
   shapeType=Capsule
   length=33.1081
   radius=14.1636
   shapeName=dyn_robeSideFront_05_l_capsule
   center=<3.2829 -7.8085 -0.425>
   rotation=<rot -8.8651 -89.5938 -2.6352>
  }]
 dyn_robeSideFront_04_r [{
   shapeType=Capsule
   length=33.1081
   radius=14.1636
   shapeName=dyn_robeSideFront_05_l_capsule
   center=<3.2823 -7.7174 -0.4108>
   rotation=<rot -8.8651 -89.5938 -2.6352>
  }]
 dyn_robeSideFront_05_l [{
   shapeType=Capsule
   length=33.1081
   radius=14.1636
   shapeName=dyn_robeSideFront_05_l_capsule
   center=<3.2706 -6.0753 -0.1547>
   rotation=<rot -8.8651 -89.5938 -2.6352>
  }]
 dyn_robeSideFront_05_r [{
   shapeType=Capsule
   length=33.1081
   radius=14.1636
   shapeName=dyn_robeSideFront_05_l_capsule
   center=<3.2666 -5.5062 -0.0659>
   rotation=<rot -8.8651 -89.5938 -2.6352>
  }]
 dyn_robeSide_01_l [{
   shapeType=Capsule
   length=33.1081
   radius=14.1636
   shapeName=dyn_robeSideFront_05_l_capsule
   center=<3.2636 -5.0836 0>
   rotation=<rot -8.8651 -89.5938 -2.6352>
  }]
 dyn_robeSide_01_r [{
   shapeType=Capsule
   length=33.1081
   radius=14.1636
   shapeName=dyn_robeSideFront_05_l_capsule
   center=<3.2636 -5.0836 0>
   rotation=<rot -8.8651 -89.5938 -2.6352>
  }]
 dyn_robeSide_02_l [{
   shapeType=Capsule
   length=33.1081
   radius=14.1636
   shapeName=dyn_robeSideFront_05_l_capsule
   center=<3.2636 -5.0836 0>
   rotation=<rot -8.8651 -89.5938 -2.6352>
  }]
 dyn_robeSide_02_r [{
   shapeType=Capsule
   length=33.1081
   radius=14.1636
   shapeName=dyn_robeSideFront_05_l_capsule
   center=<3.2636 -5.0836 0>
   rotation=<rot -8.8651 -89.5938 -2.6352>
  }]
 dyn_robeSide_03_l [{
   shapeType=Capsule
   length=33.1081
   radius=14.1636
   shapeName=dyn_robeSideFront_05_l_capsule
   center=<3.2756 -6.7748 -0.2638>
   rotation=<rot -8.8651 -89.5938 -2.6352>
  }]
 dyn_robeSide_03_r [{
   shapeType=Capsule
   length=33.1081
   radius=14.1636
   shapeName=dyn_robeSideFront_05_l_capsule
   center=<3.2925 -9.1647 -0.6366>
   rotation=<rot -8.8651 -89.5938 -2.6352>
  }]
 dyn_robeSide_04_l [{
   shapeType=Capsule
   length=33.1081
   radius=14.1636
   shapeName=dyn_robeSideFront_05_l_capsule
   center=<3.2682 -5.7189 -0.1015>
   rotation=<rot -8.8651 -89.5938 -2.6352>
  }]
 dyn_robeSide_04_r [{
   shapeType=Capsule
   length=33.1081
   radius=14.1636
   shapeName=dyn_robeSideFront_05_l_capsule
   center=<3.2829 -7.8053 -0.4245>
   rotation=<rot -8.8651 -89.5938 -2.6352>
  }]
 dyn_robeSide_05_l [{
   shapeType=Capsule
   length=33.1081
   radius=14.1636
   shapeName=dyn_robeSideFront_05_l_capsule
   center=<3.2636 -5.0836 0>
   rotation=<rot -8.8651 -89.5938 -2.6352>
  }]
 dyn_robeSide_05_r [{
   shapeType=Capsule
   length=33.1081
   radius=14.1636
   shapeName=dyn_robeSideFront_05_l_capsule
   center=<3.2749 -6.6702 -0.2475>
   rotation=<rot -8.8651 -89.5938 -2.6352>
  }]
 foot_l [{shapeType=Capsule, length=15.3736, radius=6.9293, center=<-4.9239 6.576 -0.1019>, rotation=<rot 0.1943 0.8195 -88.2017>}]
 foot_r [{shapeType=Capsule, length=15.3736, radius=6.9293, center=<4.9239 -6.576 0.1019>, rotation=<rot 0.1943 0.8195 91.7983>}]
 ground_collision [{shapeType=Sphere, radius=500, shapeName=root_sphere, center=<0 17.4691 -496.4794>}]
 thigh_l [{shapeType=Capsule, length=28.4967, radius=7.6664, center=<-20.3684 -1.3372 0.7921>, rotation=<rot 0 90.6545 -89.5769>}]
 thigh_r [{shapeType=Capsule, length=28.4967, radius=7.6664, center=<20.3684 1.3372 -0.7921>, rotation=<rot 0 90.6545 90.4231>}]
}
constraints [
 {
  bone1Name=dyn_robeBack_02
  bone2Name=dyn_robeBack_01
  swing1LimitDegrees=5
  swing2LimitDegrees=45
  swing2Motion=Locked
  twistLimitDegrees=45
  twistMotion=Locked
 }
 {bone1Name=dyn_robeBack_03, bone2Name=dyn_robeBack_02, swing1LimitDegrees=14, swing2LimitDegrees=15, twistLimitDegrees=45, twistMotion=Locked}
 {bone1Name=dyn_robeBack_04, bone2Name=dyn_robeBack_03, swing1LimitDegrees=22, swing2LimitDegrees=15, twistLimitDegrees=45, twistMotion=Locked}
 {bone1Name=dyn_robeBack_05, bone2Name=dyn_robeBack_04, swing1LimitDegrees=25, swing2LimitDegrees=8, twistLimitDegrees=45, twistMotion=Locked}
 {
  bone1Name=dyn_robeSideBack_02_l
  bone2Name=dyn_robeSideBack_01_l
  swing1LimitDegrees=5
  swing2LimitDegrees=45
  swing2Motion=Locked
  twistLimitDegrees=45
  twistMotion=Locked
 }
 {
  bone1Name=dyn_robeSideBack_03_l
  bone2Name=dyn_robeSideBack_02_l
  swing1LimitDegrees=14
  swing2LimitDegrees=15
  twistLimitDegrees=45
  twistMotion=Locked
 }
 {
  bone1Name=dyn_robeSideBack_04_l
  bone2Name=dyn_robeSideBack_03_l
  swing1LimitDegrees=22
  swing2LimitDegrees=15
  twistLimitDegrees=45
  twistMotion=Locked
 }
 {
  bone1Name=dyn_robeSideBack_05_l
  bone2Name=dyn_robeSideBack_04_l
  swing1LimitDegrees=25
  swing2LimitDegrees=8
  twistLimitDegrees=45
  twistMotion=Locked
 }
 {
  bone1Name=dyn_robeSideBack_02_r
  bone2Name=dyn_robeSideBack_01_r
  swing1LimitDegrees=5
  swing2LimitDegrees=45
  swing2Motion=Locked
  twistLimitDegrees=45
  twistMotion=Locked
 }
 {
  bone1Name=dyn_robeSideBack_03_r
  bone2Name=dyn_robeSideBack_02_r
  swing1LimitDegrees=14
  swing2LimitDegrees=15
  twistLimitDegrees=45
  twistMotion=Locked
 }
 {
  bone1Name=dyn_robeSideBack_04_r
  bone2Name=dyn_robeSideBack_03_r
  swing1LimitDegrees=22
  swing2LimitDegrees=15
  twistLimitDegrees=45
  twistMotion=Locked
 }
 {
  bone1Name=dyn_robeSideBack_05_r
  bone2Name=dyn_robeSideBack_04_r
  swing1LimitDegrees=25
  swing2LimitDegrees=8
  twistLimitDegrees=45
  twistMotion=Locked
 }
 {
  bone1Name=dyn_robeSideFront_02_l
  bone2Name=dyn_robeSideFront_01_l
  swing1LimitDegrees=30
  swing2LimitDegrees=45
  swing2Motion=Locked
  twistLimitDegrees=45
  twistMotion=Locked
 }
 {
  bone1Name=dyn_robeSideFront_03_l
  bone2Name=dyn_robeSideFront_02_l
  swing1LimitDegrees=25
  swing2LimitDegrees=15
  twistLimitDegrees=45
  twistMotion=Locked
 }
 {
  bone1Name=dyn_robeSideFront_04_l
  bone2Name=dyn_robeSideFront_03_l
  swing1LimitDegrees=22
  swing2LimitDegrees=15
  twistLimitDegrees=45
  twistMotion=Locked
 }
 {
  bone1Name=dyn_robeSideFront_05_l
  bone2Name=dyn_robeSideFront_04_l
  swing1LimitDegrees=25
  swing2LimitDegrees=8
  twistLimitDegrees=45
  twistMotion=Locked
 }
 {
  bone1Name=dyn_robeSideFront_02_r
  bone2Name=dyn_robeSideFront_01_r
  swing1LimitDegrees=5
  swing2LimitDegrees=45
  swing2Motion=Locked
  twistLimitDegrees=45
  twistMotion=Locked
 }
 {
  bone1Name=dyn_robeSideFront_03_r
  bone2Name=dyn_robeSideFront_02_r
  swing1LimitDegrees=14
  swing2LimitDegrees=15
  twistLimitDegrees=45
  twistMotion=Locked
 }
 {
  bone1Name=dyn_robeSideFront_04_r
  bone2Name=dyn_robeSideFront_03_r
  swing1LimitDegrees=22
  swing2LimitDegrees=15
  twistLimitDegrees=45
  twistMotion=Locked
 }
 {
  bone1Name=dyn_robeSideFront_05_r
  bone2Name=dyn_robeSideFront_04_r
  swing1LimitDegrees=25
  swing2LimitDegrees=8
  twistLimitDegrees=45
  twistMotion=Locked
 }
 {
  bone1Name=dyn_robeSide_02_l
  bone2Name=dyn_robeSide_01_l
  swing1LimitDegrees=5
  swing2LimitDegrees=45
  swing2Motion=Locked
  twistLimitDegrees=45
  twistMotion=Locked
 }
 {
  bone1Name=dyn_robeSide_03_l
  bone2Name=dyn_robeSide_02_l
  swing1LimitDegrees=14
  swing2LimitDegrees=15
  twistLimitDegrees=45
  twistMotion=Locked
 }
 {
  bone1Name=dyn_robeSide_04_l
  bone2Name=dyn_robeSide_03_l
  swing1LimitDegrees=22
  swing2LimitDegrees=15
  twistLimitDegrees=45
  twistMotion=Locked
 }
 {
  bone1Name=dyn_robeSide_05_l
  bone2Name=dyn_robeSide_04_l
  swing1LimitDegrees=25
  swing2LimitDegrees=8
  twistLimitDegrees=45
  twistMotion=Locked
 }
 {
  bone1Name=dyn_robeSide_02_r
  bone2Name=dyn_robeSide_01_r
  swing1LimitDegrees=5
  swing2LimitDegrees=45
  swing2Motion=Locked
  twistLimitDegrees=45
  twistMotion=Locked
 }
 {
  bone1Name=dyn_robeSide_03_r
  bone2Name=dyn_robeSide_02_r
  swing1LimitDegrees=14
  swing2LimitDegrees=15
  twistLimitDegrees=45
  twistMotion=Locked
 }
 {
  bone1Name=dyn_robeSide_04_r
  bone2Name=dyn_robeSide_03_r
  swing1LimitDegrees=22
  swing2LimitDegrees=15
  twistLimitDegrees=45
  twistMotion=Locked
 }
 {
  bone1Name=dyn_robeSide_05_r
  bone2Name=dyn_robeSide_04_r
  swing1LimitDegrees=25
  swing2LimitDegrees=8
  twistLimitDegrees=45
  twistMotion=Locked
 }
]
