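# Physics asset for the back scarf on the male medium DroveRay_Dip player body.
# Five scarfBack capsule bodies form a constrained chain from a tiny base_scarfBack anchor, with limited swing on both axes and twist locked.
# Capsules on the spine and both upper arms supply body shapes around the scarf.
# Two hood bones also carry capsules but have no constraints in this asset.
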PhysicsAsset /BRCosmetics/Characters/Player/Male/Medium/Bodies/M_MED_DroveRay_Dip/Meshes/M_MED_DroveRay_Dip_ScarfBack_Physics
bodies {
 base_scarfBack [{shapeType=Capsule, length=0.505, radius=0.505, shapeName=base_scarfBack_capsule, rotation=<rot 0 -90 0>}]
 dyn_hood_01 [
  {shapeType=Capsule, length=7.3678, radius=4.1115, shapeName=dyn_hood_01_capsule, center=<3.1996 -1.5257 0.4165>, rotation=<rot 0 -90 0>}
  {shapeType=Capsule, length=5.2379, radius=3.5392, shapeName=dyn_hood_01_capsule, center=<2.8615 0.4515 0>}
 ]
 dyn_hood_02 [{
   shapeType=Capsule
   length=3.3616
   radius=2.4929
   shapeName=dyn_hood_02_capsule
   center=<3.4073 -0.9551 1.0593>
   rotation=<rot -2.0009 -90.0002 0>
  }]
 dyn_scarfBack_01 [{
   shapeType=Capsule
   length=7.9335
   radius=5.7392
   shapeName=dyn_scarfBack_01_capsule
   center=<2.8547 -4.2117 -0.7042>
   rotation=<rot 3.3836 81.2484 5.684>
  }]
 dyn_scarfBack_02 [{
   shapeType=Capsule
   length=7.9335
   radius=5.7392
   shapeName=dyn_scarfBack_01_capsule
   center=<2.8194 -4.3551 -0.5809>
   rotation=<rot -0.9062 80.8228 5.6747>
  }]
 dyn_scarfBack_03 [{
   shapeType=Capsule
   length=7.9335
   radius=5.7392
   shapeName=dyn_scarfBack_01_capsule
   center=<3.5 -3.5374 -0.4732>
   rotation=<rot 9.5412 81.8692 5.7539>
  }]
 dyn_scarfBack_04 [{
   shapeType=Capsule
   length=7.9335
   radius=5.7392
   shapeName=dyn_scarfBack_01_capsule
   center=<3.025 -3.0938 -0.6191>
   rotation=<rot 11.0046 82.0188 5.7807>
  }]
 dyn_scarfBack_05 [{
   shapeType=Capsule
   length=7.9335
   radius=5.7392
   shapeName=dyn_scarfBack_01_capsule
   center=<3.0412 -3.0408 -0.6957>
   rotation=<rot 11.0046 82.0188 5.7807>
  }]
 spine_01 [{shapeType=Capsule, length=21.719, radius=10.0413, shapeName=spine_01_capsule, center=<0.1708 -7.8639 -0.0828>}]
 spine_02 [{shapeType=Capsule, length=18.7175, radius=10.7291, shapeName=spine_02_capsule, center=<0.7276 -5.7482 0>}]
 spine_03 [
  {shapeType=Capsule, length=18.9456, radius=9.9821, shapeName=spine_03_capsule, center=<0.9794 -5.4818 0>}
  {shapeType=Capsule, length=15, radius=15, shapeName=spine_03_capsule, center=<0 1.8054 0>}
 ]
 spine_04 [{shapeType=Capsule, length=18.0335, radius=11.5888, shapeName=spine_04_capsule, center=<-0.0903 -3.2791 0>}]
 spine_05 [{shapeType=Capsule, length=10.6206, radius=12.1582, shapeName=spine_05_capsule, center=<0.7999 -1.3338 2.1696>}]
 upperarm_l [{shapeType=Capsule, length=24.4035, radius=8.3266, center=<13.1893 0.5445 -1.6166>, rotation=<rot 0.6246 -85.425 -85.16>}]
 upperarm_r [{shapeType=Capsule, length=24.6539, radius=8.2514, center=<-13.3136 1.0087 1.5996>, rotation=<rot 0.6246 -85.425 94.84>}]
}
constraints [
 {bone1Name=dyn_scarfBack_01, bone2Name=base_scarfBack, swing1LimitDegrees=45, swing2LimitDegrees=20, twistLimitDegrees=25, twistMotion=Locked}
 {
  bone1Name=dyn_scarfBack_02
  bone2Name=dyn_scarfBack_01
  swing1LimitDegrees=40
  swing2LimitDegrees=10
  twistLimitDegrees=15
  twistMotion=Locked
 }
 {
  bone1Name=dyn_scarfBack_03
  bone2Name=dyn_scarfBack_02
  swing1LimitDegrees=35
  swing2LimitDegrees=10
  twistLimitDegrees=45
  twistMotion=Locked
 }
 {
  bone1Name=dyn_scarfBack_04
  bone2Name=dyn_scarfBack_03
  swing1LimitDegrees=35
  swing2LimitDegrees=10
  twistLimitDegrees=45
  twistMotion=Locked
 }
 {
  bone1Name=dyn_scarfBack_05
  bone2Name=dyn_scarfBack_04
  swing1LimitDegrees=35
  swing2LimitDegrees=10
  twistLimitDegrees=45
  twistMotion=Locked
 }
]
