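# Physics asset for the front scarf on the M_MED_DroveRay_Dip male medium player body cosmetic, defining its bone chain and body colliders.
# Four dyn_scarfFront capsules form a chain from a tiny base_scarfFront capsule, each joint having tight swing limits and locked twist.
# Unconstrained capsules on the spine, upper arms and hands supply body shapes for the scarf to collide with; the lower-arm bodies have no shapes.
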PhysicsAsset /BRCosmetics/Characters/Player/Male/Medium/Bodies/M_MED_DroveRay_Dip/Meshes/M_MED_DroveRay_Dip_ScarfFront_Physics
bodies {
 base_scarfFront [{shapeType=Capsule, length=0.505, radius=0.505, shapeName=base_scarfFront_capsule, rotation=<rot 0 -90 0>}]
 dyn_scarfFront_01 [{
   shapeType=Capsule
   length=6.028
   radius=2.8071
   shapeName=dyn_scarfFront_01_capsule
   center=<4.0784 -1.7382 -0.448>
   rotation=<rot 11.4583 -88.5966 88.5746>
  }]
 dyn_scarfFront_02 [{
   shapeType=Capsule
   length=6.028
   radius=2.8071
   shapeName=dyn_scarfFront_01_capsule
   center=<4.0917 -1.7379 -0.4477>
   rotation=<rot 11.4507 -88.9032 88.5138>
  }]
 dyn_scarfFront_03 [{
   shapeType=Capsule
   length=6.028
   radius=2.8071
   shapeName=dyn_scarfFront_01_capsule
   center=<4.3342 -1.8359 -0.4205>
   rotation=<rot 11.2402 -94.8376 87.3454>
  }]
 dyn_scarfFront_04 [{
   shapeType=Capsule
   length=2.9066
   radius=2.8071
   shapeName=dyn_scarfFront_01_capsule
   center=<2.5072 -1.8836 -0.4663>
   rotation=<rot 11.4284 -89.7353 88.3488>
  }]
 hand_l [{shapeType=Capsule, length=9.2245, radius=5.7859, center=<7.6765 -2.6238 -0.1092>, rotation=<rot 0.0308 87.5844 -90.3163>}]
 hand_r [{shapeType=Capsule, length=9.2245, radius=5.7859, center=<-7.6765 2.6238 0.1092>, rotation=<rot 0.0308 87.5844 89.6837>}]
 lowerarm_l []
 lowerarm_r []
 spine_01 [{shapeType=Capsule, length=16.8101, radius=9.678, center=<5.4152 2.9784 0>, rotation=<rot 0 -90 0>}]
 spine_02 [{shapeType=Capsule, length=16.3202, radius=9.678, center=<5.4152 2.9287 0>, rotation=<rot 0 -90 0>}]
 spine_03 [{shapeType=Capsule, length=16.0433, radius=9.678, center=<5.4152 2.7836 0>, rotation=<rot 0 -90 0>}]
 spine_04 [{shapeType=Capsule, length=13.8874, radius=11.1372, center=<3.3458 1.2577 0>, rotation=<rot 0 -90 0>}]
 spine_05 [{shapeType=Capsule, length=15.479, radius=8.3076, center=<-0.0861 3.0488 0>, rotation=<rot 0 -90 0>}]
 upperarm_l [{shapeType=Capsule, length=24.4035, radius=8.3266, center=<13.1893 0.5445 -1.6166>, rotation=<rot 0.6246 -85.425 -85.16>}]
 upperarm_r [{shapeType=Capsule, length=24.6539, radius=7.4974, center=<-13.1893 -0.5445 1.6166>, rotation=<rot 0.6246 -85.425 94.84>}]
}
constraints [
 {
  bone1Name=dyn_scarfFront_02
  bone2Name=dyn_scarfFront_01
  swing1LimitDegrees=6
  swing2LimitDegrees=5
  twistLimitDegrees=45
  twistMotion=Locked
 }
 {bone1Name=dyn_scarfFront_01, bone2Name=base_scarfFront, swing1LimitDegrees=6, swing2LimitDegrees=5, twistLimitDegrees=45, twistMotion=Locked}
 {
  bone1Name=dyn_scarfFront_03
  bone2Name=dyn_scarfFront_02
  swing1LimitDegrees=6
  swing2LimitDegrees=5
  twistLimitDegrees=45
  twistMotion=Locked
 }
 {
  bone1Name=dyn_scarfFront_04
  bone2Name=dyn_scarfFront_03
  swing1LimitDegrees=6
  swing2LimitDegrees=5
  twistLimitDegrees=45
  twistMotion=Locked
 }
]
