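# Post-process-style AnimBP for the DroveRay_Dip character body, adding secondary motion to robe, hood, sleeves and scarf.
# Copied mesh pose feeds a Control Rig, then pelvis RigidBody when cloth is off, then constraints, AnimDynamics and scarf RigidBody.
# Backblings swap back-scarf simulation for weighted constraints; disabled RigidBody falls back to movement-weighted robe trail controllers.
# The EventGraph updates gravity, the cloth controller, movement alpha and gating flags.
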AnimBlueprint /BRCosmetics/Characters/Player/Male/Medium/Bodies/M_MED_DroveRay_Dip/Meshes/M_MED_DroveRay_Dip_AnimBP
target_skeleton=&/M_MED_DroveRay_Dip_Skeleton
graphs [
 AnimGraph {nodes [
   n1 = RigidBody!(ComponentPose=Local To Component(LocalPose=Use cached pose 'animDynCache'()), Alpha, OverrideWorldGravity:Vector="0,0,0", ExternalForce) -> (Pose) {
    bClampLinearTranslationLimitToRefPose=true
    bOverrideWorldGravity=true
    overridePhysicsAsset=&/M_MED_DroveRay_Dip_ScarfBack_Physics
    simulationSpace=BaseBoneSpace
    alphaBoolBlend {blendInTime=2, blendOption=Linear, blendOutTime=0.5}
    alphaScaleBiasClamp {bMapRange=true}
    baseBoneRef {boneName=spine_04}
    simSpaceSettings {
     dampingAlpha=0
     maxAngularAcceleration=28
     maxAngularVelocity=7
     maxLinearAcceleration=900
     maxLinearVelocity=600
     velocityScaleZ=0.4
     worldAlpha=1
     externalLinearDragV=<0.3 0.3 0.3>
    }
   }
   n2 = RigidBody(ComponentPose=@n1, OverrideWorldGravity="0,0,0") {
    bClampLinearTranslationLimitToRefPose=true
    bOverrideWorldGravity=true
    overridePhysicsAsset=&/M_MED_DroveRay_Dip_ScarfFront_Physics
    simulationSpace=BaseBoneSpace
    alphaBoolBlend {blendInTime=2, blendOption=Linear, blendOutTime=0.5}
    alphaScaleBiasClamp {bMapRange=true}
    baseBoneRef {boneName=spine_04}
    simSpaceSettings {
     dampingAlpha=0
     maxAngularAcceleration=28
     maxAngularVelocity=7
     maxLinearAcceleration=900
     maxLinearVelocity=600
     velocityScaleZ=0.4
     worldAlpha=1
     externalLinearDragV=<0.3 0.3 0.3>
    }
   }
   n3 = "Transform (Modify) Bone - Bone: dyn_scarfBack_01"!(ComponentPose=@n2, bAlphaBoolEnabled:bool=True, Translation="-0.028,-0.486,0.059", Rotation="-0.000,-3.400,6.776", Scale="1,1,1") -> (Pose) {
    alphaInputType=Bool
    rotationMode=BMM_Additive
    rotationSpace=BCS_BoneSpace
    translationMode=BMM_Additive
    translationSpace=BCS_BoneSpace
    alphaBoolBlend {blendOption=Linear}
    boneToModify {boneName=dyn_scarfBack_01}
    rotation=<rot 0 -3.3999 6.776>
    translation=<-0.028 -0.486 0.059>
   }
   n4 = "Transform (Modify) Bone - Bone: dyn_scarfBack_02"!(ComponentPose=@n3, bAlphaBoolEnabled:bool=True, Translation="0,-1,0", Rotation="0,-3.256,0", Scale="1,1,1") -> (Pose) {
    alphaInputType=Bool
    rotationMode=BMM_Additive
    rotationSpace=BCS_BoneSpace
    translationMode=BMM_Additive
    translationSpace=BCS_BoneSpace
    alphaBoolBlend {blendInTime=0.1, blendOption=Linear, blendOutTime=0.2}
    boneToModify {boneName=dyn_scarfBack_02}
    rotation {yaw=-3.256}
    translation {y=-1}
   }
   n5 = "Transform (Modify) Bone - Bone: dyn_scarfBack_02"(ComponentPose=@n4, bAlphaBoolEnabled=True, Translation="0,0,0", Rotation="0,-8.256,0", Scale="1,1,1") {
    alphaInputType=Bool
    rotationMode=BMM_Additive
    rotationSpace=BCS_BoneSpace
    translationMode=BMM_Additive
    translationSpace=BCS_BoneSpace
    alphaBoolBlend {blendInTime=0.1, blendOption=Linear, blendOutTime=0.2}
    boneToModify {boneName=dyn_scarfBack_02}
    rotation {yaw=-8.256}
   }
   n6 = "Transform (Modify) Bone - Bone: dyn_scarfBack_03"!(ComponentPose=@n5, bAlphaBoolEnabled:bool=True, Translation="0.000,2.440,-0.000", Rotation="0,0,0", Scale="1,1,1") -> (Pose) {
    alphaInputType=Bool
    rotationMode=BMM_Additive
    rotationSpace=BCS_BoneSpace
    translationMode=BMM_Additive
    translationSpace=BCS_BoneSpace
    alphaBoolBlend {blendOption=Linear}
    boneToModify {boneName=dyn_scarfBack_03}
    translation {y=2.44}
   }
   n7 = Component To Local(ComponentPose=@n6)
   n8 = "Transform (Modify) Bone - Bone: attach_backpack"(
    ComponentPose=Local To Component(LocalPose=Use cached pose 'animDynCache'())
    Alpha=1
    Translation="0,-1.2,0"
    Rotation="0,0,8"
    Scale="1,1,1"
   ) {
    rotationMode=BMM_Additive
    rotationSpace=BCS_BoneSpace
    translationMode=BMM_Additive
    translationSpace=BCS_BoneSpace
    boneToModify {boneName=attach_backpack}
    rotation {roll=8}
    translation {y=-1.2}
   }
   n9 = "Constraint - dyn_scarfBack_01 (PPPPPPPP)"!(
    ComponentPose=@n8
    Alpha=1
    ConstraintWeights_0=0.0109
    ConstraintWeights_1:float=0.0139
    ConstraintWeights_2:float=0.0203
    ConstraintWeights_3:float=0.0504
    ConstraintWeights_4:float=0.0508
    ConstraintWeights_5:float=0.0622
    ConstraintWeights_6:float=0.3524
    ConstraintWeights_7:float=0.4392
   ) -> (Pose) {
    boneToModify {boneName=dyn_scarfBack_01}
    constraintSetup [
     {targetBone {boneName=spine_04}}
     {targetBone {boneName=deform_upperarm_bk_r}}
     {targetBone {boneName=neck_02}}
     {targetBone {boneName=deform_upperarm_up_r}}
     {targetBone {boneName=spine_03}}
     {targetBone {boneName=upperarm_twist_01_r}}
     {targetBone {boneName=clavicle_r}}
     {targetBone {boneName=spine_05}}
    ]
    constraintWeights [1,1,1,1,1,1,1,1]
   }
   n10 = "Constraint - dyn_scarfBack_02 (PPPP)"!(
    ComponentPose=@n9
    Alpha=1
    ConstraintWeights_0=0.883
    ConstraintWeights_1:float=0.0696
    ConstraintWeights_2:float=0.0306
    ConstraintWeights_3:float=0.0168
   ) -> (Pose) {
    boneToModify {boneName=dyn_scarfBack_02}
    constraintSetup [
     {targetBone {boneName=spine_05}}
     {targetBone {boneName=clavicle_r}}
     {targetBone {boneName=spine_03}}
     {targetBone {boneName=upperarm_twist_01_r}}
    ]
    constraintWeights [1,1,1,1]
   }
   n11 = "Constraint - dyn_scarfBack_03 (PPPPPPPP)"!(
    ComponentPose=@n10
    Alpha=1
    ConstraintWeights_0=0.2319
    ConstraintWeights_1:float=0.0124
    ConstraintWeights_2:float=0.0037
    ConstraintWeights_3:float=0.1222
    ConstraintWeights_4:float=0.23
    ConstraintWeights_5:float=0.012
    ConstraintWeights_6:float=0.2475
    ConstraintWeights_7:float=0.1402
   ) -> (Pose) {
    boneToModify {boneName=dyn_scarfBack_03}
    constraintSetup [
     {targetBone {boneName=spine_04}}
     {targetBone {boneName=deform_upperarm_bk_r}}
     {targetBone {boneName=spine_02}}
     {targetBone {boneName=dyn_hood_01}}
     {targetBone {boneName=spine_03}}
     {targetBone {boneName=upperarm_twist_01_r}}
     {targetBone {boneName=clavicle_r}}
     {targetBone {boneName=spine_05}}
    ]
    constraintWeights [1,1,1,1,1,1,1,1]
   }
   n12 = "Constraint - dyn_scarfBack_04 (PPPPPPPP)"!(
    ComponentPose=@n11
    Alpha=1
    ConstraintWeights_0=0.32
    ConstraintWeights_1:float=0.0061
    ConstraintWeights_2:float=0.0368
    ConstraintWeights_3:float=0.1034
    ConstraintWeights_4:float=0.3465
    ConstraintWeights_5:float=0.0029
    ConstraintWeights_6:float=0.1291
    ConstraintWeights_7:float=0.0552
   ) -> (Pose) {
    boneToModify {boneName=dyn_scarfBack_04}
    constraintSetup [
     {targetBone {boneName=spine_04}}
     {targetBone {boneName=deform_upperarm_bk_r}}
     {targetBone {boneName=spine_01}}
     {targetBone {boneName=spine_02}}
     {targetBone {boneName=spine_03}}
     {targetBone {boneName=upperarm_twist_01_r}}
     {targetBone {boneName=clavicle_r}}
     {targetBone {boneName=spine_05}}
    ]
    constraintWeights [1,1,1,1,1,1,1,1]
   }
   n13 = "Constraint - dyn_scarfBack_04 (PPPPPPPS)"!(
    ComponentPose=@n12
    Alpha=1
    ConstraintWeights_0=0.1366
    ConstraintWeights_1:float=0.0001
    ConstraintWeights_2:float=0.1114
    ConstraintWeights_3:float=0.2493
    ConstraintWeights_4:float=0.35
    ConstraintWeights_5:float=0.0709
    ConstraintWeights_6:float=0.0758
    ConstraintWeights_7:float=0.006
   ) -> (Pose) {
    boneToModify {boneName=dyn_scarfBack_04}
    constraintSetup [
     {targetBone {boneName=spine_04}}
     {targetBone {boneName=deform_upperarm_bk_r}}
     {targetBone {boneName=spine_01}}
     {targetBone {boneName=spine_02}}
     {targetBone {boneName=spine_03}}
     {targetBone {boneName=pelvis}}
     {targetBone {boneName=clavicle_r}}
     {transformType=Scale, targetBone {boneName=spine_05}}
    ]
    constraintWeights [1,1,1,1,1,1,1,1]
   }
   n14 = RigidBody(ComponentPose=@n13, OverrideWorldGravity="0,0,0") {
    bClampLinearTranslationLimitToRefPose=true
    bOverrideWorldGravity=true
    overridePhysicsAsset=&M_MED_DroveRay_Dip_ScarfFront_Physics
    simulationSpace=BaseBoneSpace
    alphaBoolBlend {blendInTime=2, blendOption=Linear, blendOutTime=0.5}
    alphaScaleBiasClamp {bMapRange=true}
    baseBoneRef {boneName=spine_04}
    simSpaceSettings {
     dampingAlpha=0
     maxAngularAcceleration=28
     maxAngularVelocity=7
     maxLinearAcceleration=900
     maxLinearVelocity=600
     velocityScaleZ=0.4
     worldAlpha=1
     externalLinearDragV=<0.3 0.3 0.3>
    }
   }
   n15 = Component To Local(ComponentPose=@n14)
   n16 = Get isWearingBackbling / Cape / NoRBAN() -> (isWearingBackbling / Cape / NoRBAN:bool)
   Output Pose(Result=Blend Poses by bool(
     BlendPose_0=Blend Poses by bool(BlendPose_0=@n15, BlendPose_1=@n7, bActiveValue=@n16) {blendType=Linear}
     BlendPose_1=Use cached pose 'trailControlerCache'()
     bActiveValue=Get bRigidBodyEnabled()
    ) {blendType=Linear}) {layerGroup=None}
   n17 = Use cached pose 'preSimCache'()
   n18 = Use cached pose 'RBANCache'()
   n19 = Get bClothEnabled()
   n20 = "Constraint - base_hood (PPPPPPPP)"!(
    ComponentPose=Local To Component(LocalPose=Blend Poses by bool(BlendPose_0=@n17, BlendPose_1=@n18, bActiveValue=@n19) {blendType=Linear})
    Alpha=1
    ConstraintWeights_0=0.3596
    ConstraintWeights_1:float=0.3531
    ConstraintWeights_2:float=0.1692
    ConstraintWeights_3:float=0.0589
    ConstraintWeights_4:float=0.0406
    ConstraintWeights_5:float=0.0174
    ConstraintWeights_6:float=0.0007
    ConstraintWeights_7:float=0.0005
   ) -> (Pose) {
    boneToModify {boneName=base_hood}
    constraintSetup [
     {targetBone {boneName=spine_03}}
     {targetBone {boneName=spine_05}}
     {targetBone {boneName=spine_04}}
     {targetBone {boneName=clavicle_l}}
     {targetBone {boneName=clavicle_r}}
     {targetBone {boneName=spine_02}}
     {targetBone {boneName=upperarm_twist_01_l}}
     {targetBone {boneName=deform_upperarm_bk_l}}
    ]
    constraintWeights [1,1,1,1,1,1,1,1]
   }
   n21 = "Constraint - dyn_hood_02 (PPPPPPPP)"!(
    ComponentPose=@n20
    bAlphaBoolEnabled:bool=True
    ConstraintWeights_0=0.5575
    ConstraintWeights_1:float=0.3523
    ConstraintWeights_2:float=0.0356
    ConstraintWeights_3:float=0.0244
    ConstraintWeights_4:float=0.0145
    ConstraintWeights_5:float=0.0077
    ConstraintWeights_6:float=0.0057
    ConstraintWeights_7:float=0.0022
   ) -> (Pose) {
    alphaInputType=Bool
    boneToModify {boneName=dyn_hood_02}
    constraintSetup [
     {targetBone {boneName=dyn_hood_02}}
     {targetBone {boneName=dyn_hood_01}}
     {targetBone {boneName=spine_03}}
     {targetBone {boneName=spine_04}}
     {targetBone {boneName=spine_05}}
     {targetBone {boneName=spine_02}}
     {targetBone {boneName=clavicle_r}}
     {targetBone {boneName=clavicle_l}}
    ]
    constraintWeights [1,1,1,1,1,1,1,1]
   }
   n22 = "Transform (Modify) Bone - Bone: base_hood"(ComponentPose=@n21, Alpha=1, Translation="0,0,0", Rotation="0,3,0", Scale="1,1,1") {
    rotationMode=BMM_Additive
    rotationSpace=BCS_BoneSpace
    alphaBoolBlend {blendOption=Linear}
    boneToModify {boneName=base_hood}
    rotation {yaw=3}
   }
   n23 = "Transform (Modify) Bone - Bone: base_hood"!(ComponentPose=@n22, bAlphaBoolEnabled:bool=True, Translation="0,0,0", Rotation="0,10,0", Scale="1,1,1") -> (Pose) {
    alphaInputType=Bool
    rotationMode=BMM_Additive
    rotationSpace=BCS_BoneSpace
    alphaBoolBlend {blendInTime=0.3, blendOption=Linear, blendOutTime=0.4}
    boneToModify {boneName=base_hood}
    rotation {yaw=10}
   }
   n24 = Anim Dynamics - Bone: dyn_hood_01!(ComponentPose=@n23, bAlphaBoolEnabled:bool=True, ExternalForce="0,0,0") -> (Pose) {
    alphaInputType=Bool
    angularDampingOverride=1
    angularSpringConstant=8000
    bAngularSpring=true
    bOverrideAngularDamping=true
    bUseGravityOverride=true
    gravityScale=0.35
    simulationSpace=RootRelative
    alphaBoolBlend {blendOption=Linear}
    boundBone {boneName=dyn_hood_01}
    componentAppliedLinearAccClamp=<2000 2000 2000>
    componentLinearAccScale=<0.4 0.4 0.7>
    componentLinearVelScale=<0.5 0.5 0.8>
    gravityOverride {z=-980}
    physicsBodyDefinitions [{
      boundBone {boneName=dyn_hood_01}
      boxExtents=<20 25 25>
      constraintSetup {angularLimitsMax=<0 20 45>, angularLimitsMin=<0 -20 0>, angularTarget=<1 0 0>}
      localJointOffset=<4 0 0>
     }]
    simSpaceSettings {maxAngularAcceleration=40, maxAngularVelocity=8, simSpaceAngularAlpha=1}
   }
   n25 = Anim Dynamics - Bone: dyn_hood_02!(ComponentPose=@n24, bAlphaBoolEnabled:bool=True, ExternalForce="0,0,0") -> (Pose) {
    alphaInputType=Bool
    angularDampingOverride=1
    angularSpringConstant=6000
    bAngularSpring=true
    bOverrideAngularDamping=true
    bUseGravityOverride=true
    gravityScale=0.35
    simulationSpace=RootRelative
    alphaBoolBlend {blendOption=Linear}
    boundBone {boneName=dyn_hood_02}
    componentAppliedLinearAccClamp=<2000 2000 2000>
    componentLinearAccScale=<0.4 0.4 0.7>
    componentLinearVelScale=<0.5 0.5 0.8>
    gravityOverride {z=-980}
    physicsBodyDefinitions [{
      boundBone {boneName=dyn_hood_02}
      boxExtents=<40 40 40>
      constraintSetup {angularLimitsMax=<0 8 20>, angularLimitsMin=<0 -8 0>, angularTarget=<1 0 0>}
      localJointOffset=<2 0 0>
     }]
    simSpaceSettings {maxAngularAcceleration=40, maxAngularVelocity=8, simSpaceAngularAlpha=1}
   }
   n26 = "Transform (Modify) Bone - Bone: base_sleeveTop_l"(ComponentPose=@n25, Alpha=1, Translation="0,0,0", Rotation="-2.543,0.622,4.608", Scale="1,1,1") {
    rotationMode=BMM_Additive
    rotationSpace=BCS_BoneSpace
    translationMode=BMM_Additive
    translationSpace=BCS_BoneSpace
    alphaBoolBlend {blendOption=Linear}
    boneToModify {boneName=base_sleeveTop_l}
    rotation=<rot -2.543 0.622 4.608>
   }
   n27 = "Transform (Modify) Bone - Bone: base_sleeveBottom_l"(ComponentPose=@n26, Alpha=1, Translation="-0.415,1.375,-0.405", Rotation="6.756,11.837,6.685", Scale="1,1,1") {
    rotationMode=BMM_Additive
    rotationSpace=BCS_BoneSpace
    translationMode=BMM_Additive
    translationSpace=BCS_BoneSpace
    alphaBoolBlend {blendOption=Linear}
    boneToModify {boneName=base_sleeveBottom_l}
    rotation=<rot 6.756 11.837 6.685>
    translation=<-0.415 1.375 -0.405>
   }
   n28 = "Transform (Modify) Bone - Bone: base_sleeveBack_l"(ComponentPose=@n27, Alpha=1, Translation="0,0,0", Rotation="4.080,8.355,-6.667", Scale="1,1,1") {
    rotationMode=BMM_Additive
    rotationSpace=BCS_BoneSpace
    translationMode=BMM_Additive
    translationSpace=BCS_BoneSpace
    alphaBoolBlend {blendOption=Linear}
    boneToModify {boneName=base_sleeveBack_l}
    rotation=<rot 4.0801 8.355 -6.667>
   }
   n29 = "Transform (Modify) Bone - Bone: base_sleeveFront_l"(ComponentPose=@n28, Alpha=1, Translation="0,0,0", Rotation="5.137,-0.262,0.161", Scale="1,1,1") {
    rotationMode=BMM_Additive
    rotationSpace=BCS_BoneSpace
    translationMode=BMM_Additive
    translationSpace=BCS_BoneSpace
    alphaBoolBlend {blendOption=Linear}
    boneToModify {boneName=base_sleeveFront_l}
    rotation=<rot 5.137 -0.262 0.161>
   }
   n30 = "Transform (Modify) Bone - Bone: base_sleeveTop_r"(ComponentPose=@n29, Alpha=1, Translation="1.704,0.188,-0.083", Rotation="-0.960,6.851,12.920", Scale="1,1,1") {
    rotationMode=BMM_Additive
    rotationSpace=BCS_BoneSpace
    translationMode=BMM_Additive
    translationSpace=BCS_BoneSpace
    alphaBoolBlend {blendOption=Linear}
    boneToModify {boneName=base_sleeveTop_r}
    rotation=<rot -0.9599 6.851 12.92>
    translation=<1.704 0.188 -0.083>
   }
   n31 = "Transform (Modify) Bone - Bone: base_sleeveBottom_r"(ComponentPose=@n30, Alpha=1, Translation="0,0,0", Rotation="-0.876,18.300,-19.262", Scale="1,1,1") {
    rotationMode=BMM_Additive
    rotationSpace=BCS_BoneSpace
    translationMode=BMM_Additive
    translationSpace=BCS_BoneSpace
    alphaBoolBlend {blendOption=Linear}
    boneToModify {boneName=base_sleeveBottom_r}
    rotation=<rot -0.8761 18.3 -19.2621>
   }
   n32 = "Transform (Modify) Bone - Bone: base_sleeveBack_r"(ComponentPose=@n31, Alpha=1, Translation="0,0,0", Rotation="5.422,1.573,1.495", Scale="1,1,1") {
    rotationMode=BMM_Additive
    rotationSpace=BCS_BoneSpace
    translationMode=BMM_Additive
    translationSpace=BCS_BoneSpace
    alphaBoolBlend {blendOption=Linear}
    boneToModify {boneName=base_sleeveBack_r}
    rotation=<rot 5.422 1.573 1.495>
   }
   n33 = "Transform (Modify) Bone - Bone: base_sleeveFront_r"(ComponentPose=@n32, Alpha=1, Translation="0,0,0", Rotation="-7.108,6.200,-0.989", Scale="1,1,1") {
    rotationMode=BMM_Additive
    rotationSpace=BCS_BoneSpace
    translationMode=BMM_Additive
    translationSpace=BCS_BoneSpace
    alphaBoolBlend {blendOption=Linear}
    boneToModify {boneName=base_sleeveFront_r}
    rotation=<rot -7.108 6.2 -0.989>
   }
   n34 = Anim Dynamics - Bone: dyn_sleeveTop_l(ComponentPose=@n33, Alpha=1, ExternalForce="0,0,0") {
    angularDampingOverride=1
    angularSpringConstant=1000
    bAngularSpring=true
    bOverrideAngularDamping=true
    bUseGravityOverride=true
    gravityScale=0.35
    simulationSpace=RootRelative
    alphaBoolBlend {blendOption=Linear}
    boundBone {boneName=dyn_sleeveTop_l}
    componentAppliedLinearAccClamp=<2000 2000 2000>
    componentLinearAccScale=<0.4 0.4 0.4>
    componentLinearVelScale=<0.5 0.5 0.5>
    gravityOverride {z=-980}
    physicsBodyDefinitions [{
      boundBone {boneName=dyn_sleeveTop_l}
      boxExtents=<20 20 20>
      constraintSetup {angularLimitsMax=<5 0 0>, angularLimitsMin=<-5 0 -15>, angularTarget=<1 0 0>}
      localJointOffset=<2 0 0>
     }]
    simSpaceSettings {maxAngularAcceleration=40, maxAngularVelocity=8, simSpaceAngularAlpha=1}
   }
   n35 = Anim Dynamics - Bone: dyn_sleeveBottom_l(ComponentPose=@n34, Alpha=1, ExternalForce="0,0,0") {
    angularDampingOverride=1
    angularSpringConstant=1000
    bAngularSpring=true
    bOverrideAngularDamping=true
    bUseGravityOverride=true
    gravityScale=0.35
    simulationSpace=RootRelative
    alphaBoolBlend {blendOption=Linear}
    boundBone {boneName=dyn_sleeveBottom_l}
    componentAppliedLinearAccClamp=<2000 2000 2000>
    componentLinearAccScale=<0.4 0.4 0.4>
    componentLinearVelScale=<0.5 0.5 0.5>
    gravityOverride {z=-980}
    physicsBodyDefinitions [{
      boundBone {boneName=dyn_sleeveBottom_l}
      boxExtents=<20 20 20>
      constraintSetup {angularLimitsMax=<5 0 0>, angularLimitsMin=<-5 0 -10>, angularTarget=<1 0 0>}
      localJointOffset=<2 0 0>
     }]
    simSpaceSettings {maxAngularAcceleration=40, maxAngularVelocity=8, simSpaceAngularAlpha=1}
   }
   n36 = Anim Dynamics - Bone: dyn_sleeveFront_l(ComponentPose=@n35, Alpha=1, ExternalForce="0,0,0") {
    angularDampingOverride=1
    angularSpringConstant=1000
    bAngularSpring=true
    bOverrideAngularDamping=true
    bUseGravityOverride=true
    gravityScale=0.35
    simulationSpace=RootRelative
    alphaBoolBlend {blendOption=Linear}
    boundBone {boneName=dyn_sleeveFront_l}
    componentAppliedLinearAccClamp=<2000 2000 2000>
    componentLinearAccScale=<0.4 0.4 0.4>
    componentLinearVelScale=<0.5 0.5 0.5>
    gravityOverride {z=-980}
    physicsBodyDefinitions [{
      boundBone {boneName=dyn_sleeveFront_l}
      boxExtents=<20 20 20>
      constraintSetup {angularLimitsMax=<5 0 2>, angularLimitsMin=<-5 0 -24>, angularTarget=<1 0 0>}
      localJointOffset=<2 0 0>
     }]
    simSpaceSettings {maxAngularAcceleration=40, maxAngularVelocity=8, simSpaceAngularAlpha=1}
   }
   n37 = Anim Dynamics - Bone: dyn_sleeveBack_l(ComponentPose=@n36, Alpha=1, ExternalForce="0,0,0") {
    angularDampingOverride=1
    angularSpringConstant=1000
    bAngularSpring=true
    bOverrideAngularDamping=true
    bUseGravityOverride=true
    gravityScale=0.35
    simulationSpace=RootRelative
    alphaBoolBlend {blendOption=Linear}
    boundBone {boneName=dyn_sleeveBack_l}
    componentAppliedLinearAccClamp=<2000 2000 2000>
    componentLinearAccScale=<0.4 0.4 0.4>
    componentLinearVelScale=<0.5 0.5 0.5>
    gravityOverride {z=-980}
    physicsBodyDefinitions [{
      boundBone {boneName=dyn_sleeveBack_l}
      boxExtents=<20 20 20>
      constraintSetup {angularLimitsMax=<5 8 0>, angularLimitsMin=<-5 -8 -10>, angularTarget=<1 0 0>}
      localJointOffset=<2 0 0>
     }]
    simSpaceSettings {maxAngularAcceleration=40, maxAngularVelocity=8, simSpaceAngularAlpha=1}
   }
   n38 = Anim Dynamics - Bone: dyn_sleeveTop_r(ComponentPose=@n37, Alpha=1, ExternalForce="0,0,0") {
    angularDampingOverride=1
    angularSpringConstant=1000
    bAngularSpring=true
    bOverrideAngularDamping=true
    bUseGravityOverride=true
    gravityScale=0.35
    simulationSpace=RootRelative
    alphaBoolBlend {blendOption=Linear}
    boundBone {boneName=dyn_sleeveTop_r}
    componentAppliedLinearAccClamp=<2000 2000 2000>
    componentLinearAccScale=<0.4 0.4 0.4>
    componentLinearVelScale=<0.5 0.5 0.5>
    gravityOverride {z=-980}
    physicsBodyDefinitions [{
      boundBone {boneName=dyn_sleeveTop_r}
      boxExtents=<20 20 20>
      constraintSetup {angularLimitsMax=<5 0 2>, angularLimitsMin=<-5 0 -13>, angularTarget=<1 0 0>}
      localJointOffset=<-2 0 0>
     }]
    simSpaceSettings {maxAngularAcceleration=40, maxAngularVelocity=8, simSpaceAngularAlpha=1}
   }
   n39 = Anim Dynamics - Bone: dyn_sleeveBottom_r(ComponentPose=@n38, Alpha=1, ExternalForce="0,0,0") {
    angularDampingOverride=1
    angularSpringConstant=1000
    bAngularSpring=true
    bOverrideAngularDamping=true
    bUseGravityOverride=true
    gravityScale=0.35
    simulationSpace=RootRelative
    alphaBoolBlend {blendOption=Linear}
    boundBone {boneName=dyn_sleeveBottom_r}
    componentAppliedLinearAccClamp=<2000 2000 2000>
    componentLinearAccScale=<0.4 0.4 0.4>
    componentLinearVelScale=<0.5 0.5 0.5>
    gravityOverride {z=-980}
    physicsBodyDefinitions [{
      boundBone {boneName=dyn_sleeveBottom_r}
      boxExtents=<20 20 20>
      constraintSetup {angularLimitsMax=<5 0 0>, angularLimitsMin=<-5 0 -10>, angularTarget=<1 0 0>}
      localJointOffset=<-2 0 0>
     }]
    simSpaceSettings {maxAngularAcceleration=40, maxAngularVelocity=8, simSpaceAngularAlpha=1}
   }
   n40 = Anim Dynamics - Bone: dyn_sleeveFront_r(ComponentPose=@n39, Alpha=1, ExternalForce="0,0,0") {
    angularDampingOverride=1
    angularSpringConstant=1000
    bAngularSpring=true
    bOverrideAngularDamping=true
    bUseGravityOverride=true
    gravityScale=0.35
    simulationSpace=RootRelative
    alphaBoolBlend {blendOption=Linear}
    boundBone {boneName=dyn_sleeveFront_r}
    componentAppliedLinearAccClamp=<2000 2000 2000>
    componentLinearAccScale=<0.4 0.4 0.4>
    componentLinearVelScale=<0.5 0.5 0.5>
    gravityOverride {z=-980}
    physicsBodyDefinitions [{
      boundBone {boneName=dyn_sleeveFront_r}
      boxExtents=<20 20 20>
      constraintSetup {angularLimitsMax=<5 0 1>, angularLimitsMin=<-5 0 -24>, angularTarget=<1 0 0>}
      localJointOffset=<-2 0 0>
     }]
    simSpaceSettings {maxAngularAcceleration=40, maxAngularVelocity=8, simSpaceAngularAlpha=1}
   }
   n41 = Anim Dynamics - Bone: dyn_sleeveBack_r(ComponentPose=@n40, Alpha=1, ExternalForce="0,0,0") {
    angularDampingOverride=1
    angularSpringConstant=1000
    bAngularSpring=true
    bOverrideAngularDamping=true
    bUseGravityOverride=true
    gravityScale=0.35
    simulationSpace=RootRelative
    alphaBoolBlend {blendOption=Linear}
    boundBone {boneName=dyn_sleeveBack_r}
    componentAppliedLinearAccClamp=<2000 2000 2000>
    componentLinearAccScale=<0.4 0.4 0.4>
    componentLinearVelScale=<0.5 0.5 0.5>
    gravityOverride {z=-980}
    physicsBodyDefinitions [{
      boundBone {boneName=dyn_sleeveBack_r}
      boxExtents=<20 20 20>
      constraintSetup {angularLimitsMax=<8 8 0>, angularLimitsMin=<-8 -8 -10>, angularTarget=<1 0 0>}
      localJointOffset=<-2 0 0>
     }]
    simSpaceSettings {maxAngularAcceleration=40, maxAngularVelocity=8, simSpaceAngularAlpha=1}
   }
   n42 = "Constraint - base_scarfFront (PPPPPPPP)"!(
    ComponentPose=@n41
    Alpha=1
    ConstraintWeights_0=0.0223
    ConstraintWeights_1:float=0.0252
    ConstraintWeights_2:float=0.0272
    ConstraintWeights_3:float=0.0837
    ConstraintWeights_4:float=0.1785
    ConstraintWeights_5:float=0.1926
    ConstraintWeights_6:float=0.2112
    ConstraintWeights_7:float=0.2594
   ) -> (Pose) {
    boneToModify {boneName=base_scarfFront}
    constraintSetup [
     {targetBone {boneName=upperarm_twist_01_r}}
     {targetBone {boneName=deform_upperarm_fr_r}}
     {targetBone {boneName=spine_02}}
     {targetBone {boneName=spine_03}}
     {targetBone {boneName=deform_pec_r}}
     {targetBone {boneName=spine_04}}
     {targetBone {boneName=clavicle_r}}
     {targetBone {boneName=spine_05}}
    ]
    constraintWeights [1,1,1,1,1,1,1,1]
   }
   n43 = "Constraint - base_scarfBack (PPPPPPPP)"!(
    ComponentPose=@n42
    Alpha=1
    ConstraintWeights_0=0.0109
    ConstraintWeights_1:float=0.0139
    ConstraintWeights_2:float=0.0203
    ConstraintWeights_3:float=0.0504
    ConstraintWeights_4:float=0.0508
    ConstraintWeights_5:float=0.0622
    ConstraintWeights_6:float=0.3524
    ConstraintWeights_7:float=0.4392
   ) -> (Pose) {
    boneToModify {boneName=base_scarfBack}
    constraintSetup [
     {targetBone {boneName=spine_04}}
     {targetBone {boneName=deform_upperarm_bk_r}}
     {targetBone {boneName=neck_02}}
     {targetBone {boneName=deform_upperarm_up_r}}
     {targetBone {boneName=spine_03}}
     {targetBone {boneName=upperarm_twist_01_r}}
     {targetBone {boneName=clavicle_r}}
     {targetBone {boneName=spine_05}}
    ]
    constraintWeights [1,1,1,1,1,1,1,1]
   }
   Save cached pose 'animDynCache'(Pose=Component To Local(ComponentPose=@n43)) {cacheName=animDynCache}
   n44 = Get BwdMovement() -> (BwdMovement:double)
   n45 = Trail controller - Bone: dyn_robeSideBack_05_l(ComponentPose=Local To Component(LocalPose=Use cached pose 'preSimCache'()), Alpha=@n44) {
    bUsePlanarLimit=true
    chainLength=4
    debugLifeTime=0.1
    maxDeltaTime=0.333
    alphaScaleBiasClamp {bInterpResult=true}
    rotationLimits [{limitMax=<180 180 180>}, {limitMax=<180 180 180>}, {limitMax=<180 180 180>}, {limitMax=<180 180 180>}]
    rotationOffsets [<0 0 0>, <0 0 0>, <0 0 0>, <0 0 0>]
    trailBone {boneName=dyn_robeSideBack_05_l}
    trailRelaxationSpeed {editorCurveData {keys [<at -0.0134 19.3841 RCIM_Linear>,<at 0.5079 12.8986 RCIM_Linear>,<at 0.9955 9.6044 RCIM_Linear>]}}
   }
   n46 = Trail controller - Bone: dyn_robeSideBack_05_r(ComponentPose=@n45, Alpha=@n44) {
    bUsePlanarLimit=true
    chainLength=4
    debugLifeTime=0.1
    maxDeltaTime=0.333
    alphaScaleBiasClamp {bInterpResult=true}
    rotationLimits [{limitMax=<180 180 180>}, {limitMax=<180 180 180>}, {limitMax=<180 180 180>}, {limitMax=<180 180 180>}]
    rotationOffsets [<0 0 0>, <0 0 0>, <0 0 0>, <0 0 0>]
    trailBone {boneName=dyn_robeSideBack_05_r}
    trailRelaxationSpeed {editorCurveData {keys [<at -0.0134 19.3841 RCIM_Linear>,<at 0.5079 12.8986 RCIM_Linear>,<at 0.9955 9.6044 RCIM_Linear>]}}
   }
   n47 = Trail controller - Bone: dyn_robeBack_05(ComponentPose=@n46, Alpha=@n44) {
    bUsePlanarLimit=true
    chainLength=4
    debugLifeTime=0.1
    maxDeltaTime=0.0333
    alphaScaleBiasClamp {bInterpResult=true}
    rotationLimits [{limitMax=<180 180 180>}, {limitMax=<180 180 180>}, {limitMax=<180 180 180>}, {limitMax=<180 180 180>}]
    rotationOffsets [<0 0 0>, <0 0 0>, <0 0 0>, <0 0 0>]
    trailBone {boneName=dyn_robeBack_05}
    trailRelaxationSpeed {editorCurveData {keys [<at 0 15 RCIM_Linear>,<at 0.5289 13.1838 RCIM_Linear>,<at 1 6 RCIM_Linear>]}}
   }
   n48 = Trail controller - Bone: dyn_robeSide_05_l(ComponentPose=@n47, Alpha=@n44) {
    bUsePlanarLimit=true
    chainLength=3
    debugLifeTime=0.1
    maxDeltaTime=0.0333
    alphaScaleBiasClamp {bInterpResult=true}
    rotationLimits [{limitMax=<180 180 180>}, {limitMax=<180 180 180>}, {limitMax=<180 180 180>}]
    rotationOffsets [<0 0 0>, <0 0 0>, <0 0 0>]
    trailBone {boneName=dyn_robeSide_05_l}
    trailRelaxationSpeed {editorCurveData {keys [<at 0 15 RCIM_Linear>,<at 0.5289 13.1838 RCIM_Linear>,<at 1 6 RCIM_Linear>]}}
   }
   n49 = Trail controller - Bone: dyn_robeSide_05_r(ComponentPose=@n48, Alpha=@n44) {
    bUsePlanarLimit=true
    chainLength=3
    debugLifeTime=0.1
    maxDeltaTime=0.333
    alphaScaleBiasClamp {bInterpResult=true}
    rotationLimits [{limitMax=<180 180 180>}, {limitMax=<180 180 180>}, {limitMax=<180 180 180>}]
    rotationOffsets [<0 0 0>, <0 0 0>, <0 0 0>]
    trailBone {boneName=dyn_robeSide_05_r}
    trailRelaxationSpeed {editorCurveData {keys [<at -0.0134 19.3841 RCIM_Linear>,<at 0.5079 12.8986 RCIM_Linear>,<at 0.9955 9.6044 RCIM_Linear>]}}
   }
   Save cached pose 'trailControlerCache'(Pose=Component To Local(ComponentPose=@n49)) {cacheName=trailControlerCache}
   n50 = RigidBody!(
    ComponentPose=Local To Component(LocalPose=Use cached pose 'preSimCache'())
    bAlphaBoolEnabled:bool
    OverrideWorldGravity="0,0,0"
    ExternalForce
   ) -> (Pose) {
    alphaInputType=Bool
    bOverrideWorldGravity=true
    overridePhysicsAsset=&/M_MED_DroveRay_Dip_Physics
    simulationSpace=BaseBoneSpace
    alphaBoolBlend {blendOption=Linear}
    alphaScaleBiasClamp {bMapRange=true}
    baseBoneRef {boneName=pelvis}
    componentAppliedLinearAccClamp=<4000 4000 4000>
    simSpaceSettings {
     dampingAlpha=0.4
     maxAngularAcceleration=12
     maxAngularVelocity=5
     maxLinearAcceleration=900
     maxLinearVelocity=300
     velocityScaleZ=0.4
     worldAlpha=1
    }
   }
   Save cached pose 'RBANCache'(Pose=Component To Local(ComponentPose=@n50)) {cacheName=RBANCache}
   n51 = Control Rig!(
    Source=Copy Pose From Mesh() {bUseAttachedParent=true}
    Alpha
    isCrouching_CR:bool=Get bIsCrouching()
    isBaseballSliding_CR:bool=Get bIsBaseballSliding()
    isKneeSliding_CR:bool=Get bIsKneeSliding()
    isOnGrindrail_CR:bool=Get bIsOnGrindRail()
    isDBNO_CR:bool=Get bIsDBNO()
    isRunning_CR:bool=Get isRunning() -> (isRunning:bool)
    RBANenabled_CR:bool=Get bRigidBodyEnabled()
    clothEnabled_CR:bool=Get bClothEnabled()
    isSkydiving_CR:bool=Get bIsSkyDiving()
   ) -> (Pose) {controlRigAssetReference {blueprintRigClass=&/M_MED_DroveRay_Dip_CtrlRig.M_MED_DroveRay_Dip_CtrlRig_C}}
   Save cached pose 'preSimCache'(Pose=@n51) {cacheName=preSimCache}
  ]}
 EventGraph {nodes [
   n1 = Event BlueprintUpdateAnimation()
   > Set DeltaTimeLocal(DeltaTimeLocal:double=@n1.DeltaTimeX) -> (Output_Get:double)
   > n2 = Sequence!(execute) -> (then_0, then_1, then_2:Exec, then_3:Exec)
   > n3 = "Gravity Override State Function (Parameter Asset) (Nativized)"(
    CharacterPartAnimInstance=Self-Reference()
    GameTimeInSeconds=GetGameTimeInSeconds()
    Parameters=&/M_MED_DroveRay_Dip_Parameters
   )
   > Set Gravity Override(Gravity Override=@n3)
   > Branch(Condition=Get bClothEnabled())
   > "Cloth Controller Function (Parameter Asset) (Nativized)"!(
    execute
    CharacterPartAnimInstance:Custom Character Part Anim Instance Object Reference=Self-Reference()
    ClothingAssetName:Name=DroveRay_Dip_Cloth
    ClothCoreSettings:"Cloth Core Settings Structure (by ref)"=Get Cloth Core Settings() -> (Cloth Core Settings:Cloth Core Settings Structure)
    WindGustLocalSettings:"Wind Gust Settings Structure (by ref)"=Get Wind Gust Local Settings() -> (Wind Gust Local Settings:Wind Gust Settings Structure)
    ClothUpAxis:Vector="1,0,0"
    LinVelScaleFwdBwdClamp:float=0
    LinVelScaleUpDwnClamp:float=0.7
    DeltaTime:float=Get DeltaTimeLocal() -> (DeltaTimeLocal:double)
    GameTimeInSeconds:float=GetGameTimeInSeconds()
    StateLogicLayerName:Name=none
    Parameters:Fort Character Dynamics Parameters Object Reference=&M_MED_DroveRay_Dip_Parameters
   ) -> (
    then
    OutCurrentState:Name=None
    OutClothParameters:Cloth Parameters Structure
    OutClothCoreSettings:Cloth Core Settings Structure
    OutWindGustSettings:Wind Gust Parameters Structure
    OutWindGustLocalSettings:Wind Gust Settings Structure
    OutWindVelocity:Vector
    OutLinearVelVector:Vector
   )
   n4 = PlayerMovement(execute=@n2.then_1, AnimInstance=Self-Reference(), XDotProductCoefficient=1, YDotProductCoefficient=1, ZDotProductCoefficient=1)
   n5 = MapRangeClamped(Value=@n4.YMovement, InRangeA=0, InRangeB=1)
   n6 = "Absolute (Float)"(A=@n4)
   > Set BwdMovement(BwdMovement:double=MapRangeClamped(Value=float + float(A=@n6, B=@n5) {is=title K2Node_CommutativeAssociativeBinaryOperator}, InRangeA=0, InRangeB=1.4)) -> (Output_Get:double)
   n7 = Branch(execute=@n2.then_2, Condition=Get bIsSwimSprinting() -> (bIsSwimSprinting:bool))
   n8 = Branch(execute=@n7.else, Condition=Get bIsSkyDiving())
   > n9 = Branch(Condition=Get bIsParachuteOpened())
   > Set BwdMovement(BwdMovement=float * float(A=Get BwdMovement(), B=0.4))
   Branch(execute=@n9.else, Condition="float > float"(A=Get PawnVelocityZ(), B=50) {is=title K2Node_CallFunction})
   > Set BwdMovement(BwdMovement=float * float(A=Get BwdMovement(), B=0))
   Set BwdMovement(execute=@n8.else, BwdMovement=float * float(A=Get BwdMovement(), B=1))
   n10 = Get bIsCrouchSprinting()
   n11 = Get bIsTacticalSprinting()
   n12 = Get bIsSprinting()
   Set isRunning(execute=@n2.then_3, isRunning:bool=AND Boolean(A=OR Boolean(A=@n11, B=@n12), B=NOT Boolean(A=@n10))) -> (Output_Get:bool)
   n13 = CheckForBackbling()
   n14 = OR Boolean!(A=@n13, B=@n13.bIsBackblingACape, C:bool=@n13.bIsBackblingLong, D:bool=NOT Boolean(A=Get bRigidBodyEnabled())) -> (ReturnValue)
   > Set isWearingBackbling / Cape / NoRBAN(isWearingBackbling / Cape / NoRBAN:bool=@n14) -> (Output_Get:bool)
   > Set turnOffAnimDyn(turnOffAnimDyn:bool=NOT Boolean(A=@n14)) -> (Output_Get:bool)
   n15 = Get bClothEnabled()
   > Set turnOffRBAN(turnOffRBAN:bool=AND Boolean(A=NOT Boolean(A=@n15), B=Get bRigidBodyEnabled())) -> (Output_Get:bool)
  ]}
]
variables [
 BwdMovement {type=number}
 Cloth Core Settings {
  type=ClothCoreSettings
  default {
   collisionThickness=0
   dampingCoefficient=0.05
   frictionCoefficient=0.3
   localDampingCoefficient=0
   numberIterations=2
   numberSubsteps=1
   selfCollisionThickness=0
   animDriveDamping=<0 1>
   animDriveStiffness=<0 1>
   areaStiffness=<1 1>
   bendingStiffness=<1 1>
   edgeStiffness=<1 1>
   tetherStiffness=<1 1>
  }
 }
 DeltaTimeLocal {type=number}
 Gravity Override {type=Vector}
 isRunning {type=boolean}
 isWearingBackbling / Cape / NoRBAN {type=boolean}
 turnOffAnimDyn {type=boolean}
 turnOffRBAN {type=boolean}
 Wind Gust Local Settings {type=WindGustSettings, default {elapsedTime=0, finalGustStrengthLocal=0, gustDuration=0, gustStrength=0, timeBetweenGusts=0}}
]
