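# Physics asset of collision shapes, named as cape colliders, for the male medium DroveRay Dip player cosmetic body.
# Capsules cover the clavicles, neck, pelvis and spine with a sphere on the head, while arm, leg and foot bodies have no shapes.
# Hood and five-segment robe bone chains carry near-identical capsules named after dyn_robeBack_05, and no constraints join any bodies.
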PhysicsAsset /BRCosmetics/Characters/Player/Male/Medium/Bodies/M_MED_DroveRay_Dip/Meshes/M_MED_DroveRay_Dip_CapeColliders
bodies {
 calf_l []
 calf_r []
 clavicle_l [{shapeType=Capsule, length=9.0192, radius=9.8871, center=<9.4266 -1.1806 -2.4977>, rotation=<rot 69.3297 1.4476 176.9462>}]
 clavicle_r [{shapeType=Capsule, length=9.0192, radius=9.8871, center=<-9.4266 1.1806 2.4977>, rotation=<rot 69.3297 1.4476 -3.0538>}]
 dyn_hood_01 [{shapeType=Capsule, length=15.9012, radius=6.0154, shapeName=dyn_robeBack_05_capsule, center=<6.669 -1.6614 0>, rotation=<rot 0 -90 0>}]
 dyn_hood_02 [{shapeType=Capsule, length=15.9012, radius=6.0154, shapeName=dyn_robeBack_05_capsule, center=<5.1541 -0.3216 0>, rotation=<rot 0 -90 0>}]
 dyn_robeBack_01 [{shapeType=Capsule, length=15.9012, radius=9.1867, shapeName=dyn_robeBack_05_capsule, center=<7.1694 2.3664 0>, rotation=<rot 0 -90 0>}]
 dyn_robeBack_02 [{shapeType=Capsule, length=15.9012, radius=9.1867, shapeName=dyn_robeBack_05_capsule, center=<7.1694 2.3664 0>, rotation=<rot 0 -90 0>}]
 dyn_robeBack_03 [{shapeType=Capsule, length=15.9012, radius=9.1867, shapeName=dyn_robeBack_05_capsule, center=<7.1694 2.3664 0>, rotation=<rot 0 -90 0>}]
 dyn_robeBack_04 [{shapeType=Capsule, length=15.9012, radius=9.1867, shapeName=dyn_robeBack_05_capsule, center=<7.1694 2.3664 0>, rotation=<rot 0 -90 0>}]
 dyn_robeBack_05 [{shapeType=Capsule, length=15.9012, radius=9.1867, shapeName=dyn_robeBack_05_capsule, center=<10.115 2.3664 0>, rotation=<rot 0 -90 0>}]
 dyn_robeSideBack_01_l [{shapeType=Capsule, length=15.9012, radius=9.1867, shapeName=dyn_robeBack_05_capsule, center=<7.1694 2.3664 0>, rotation=<rot 0 -90 0>}]
 dyn_robeSideBack_01_r [{shapeType=Capsule, length=15.9012, radius=9.1867, shapeName=dyn_robeBack_05_capsule, center=<7.1694 2.3664 0>, rotation=<rot 0 -90 0>}]
 dyn_robeSideBack_02_l [{shapeType=Capsule, length=15.9012, radius=9.1867, shapeName=dyn_robeBack_05_capsule, center=<7.1694 2.3664 0>, rotation=<rot 0 -90 0>}]
 dyn_robeSideBack_02_r [{shapeType=Capsule, length=15.9012, radius=9.1867, shapeName=dyn_robeBack_05_capsule, center=<7.1694 2.3664 0>, rotation=<rot 0 -90 0>}]
 dyn_robeSideBack_03_l [{shapeType=Capsule, length=15.9012, radius=9.1867, shapeName=dyn_robeBack_05_capsule, center=<7.1694 2.3664 0>, rotation=<rot 0 -90 0>}]
 dyn_robeSideBack_03_r [{shapeType=Capsule, length=15.9012, radius=9.1867, shapeName=dyn_robeBack_05_capsule, center=<7.1694 2.3664 0>, rotation=<rot 0 -90 0>}]
 dyn_robeSideBack_04_l [{shapeType=Capsule, length=15.9012, radius=9.1867, shapeName=dyn_robeBack_05_capsule, center=<7.1694 2.3664 0>, rotation=<rot 0 -90 0>}]
 dyn_robeSideBack_04_r [{shapeType=Capsule, length=15.9012, radius=9.1867, shapeName=dyn_robeBack_05_capsule, center=<7.1694 2.3664 0>, rotation=<rot 0 -90 0>}]
 dyn_robeSideBack_05_l [{shapeType=Capsule, length=15.9012, radius=9.1867, shapeName=dyn_robeBack_05_capsule, center=<10.115 2.3664 0>, rotation=<rot 0 -90 0>}]
 dyn_robeSideBack_05_r [{shapeType=Capsule, length=15.9012, radius=9.1867, shapeName=dyn_robeBack_05_capsule, center=<10.115 2.3664 0>, rotation=<rot 0 -90 0>}]
 dyn_robeSideFront_01_l [{shapeType=Capsule, length=15.9012, radius=9.1867, shapeName=dyn_robeBack_05_capsule, center=<7.1694 2.3664 0>, rotation=<rot 0 -90 0>}]
 dyn_robeSideFront_01_r [{shapeType=Capsule, length=15.9012, radius=9.1867, shapeName=dyn_robeBack_05_capsule, center=<7.1694 2.3664 0>, rotation=<rot 0 -90 0>}]
 dyn_robeSideFront_02_l [{shapeType=Capsule, length=15.9012, radius=9.1867, shapeName=dyn_robeBack_05_capsule, center=<7.1694 2.3664 0>, rotation=<rot 0 -90 0>}]
 dyn_robeSideFront_02_r [{shapeType=Capsule, length=15.9012, radius=9.1867, shapeName=dyn_robeBack_05_capsule, center=<7.1694 2.3664 0>, rotation=<rot 0 -90 0>}]
 dyn_robeSideFront_03_l [{shapeType=Capsule, length=15.9012, radius=9.1867, shapeName=dyn_robeBack_05_capsule, center=<7.1694 2.3664 0>, rotation=<rot 0 -90 0>}]
 dyn_robeSideFront_03_r [{shapeType=Capsule, length=15.9012, radius=9.1867, shapeName=dyn_robeBack_05_capsule, center=<7.1694 2.3664 0>, rotation=<rot 0 -90 0>}]
 dyn_robeSideFront_04_l [{shapeType=Capsule, length=15.9012, radius=9.1867, shapeName=dyn_robeBack_05_capsule, center=<7.1694 2.3664 0>, rotation=<rot 0 -90 0>}]
 dyn_robeSideFront_04_r [{shapeType=Capsule, length=15.9012, radius=9.1867, shapeName=dyn_robeBack_05_capsule, center=<7.1694 2.3664 0>, rotation=<rot 0 -90 0>}]
 dyn_robeSideFront_05_l [{shapeType=Capsule, length=15.9012, radius=9.1867, shapeName=dyn_robeBack_05_capsule, center=<7.1694 2.3664 0>, rotation=<rot 0 -90 0>}]
 dyn_robeSideFront_05_r [{shapeType=Capsule, length=15.9012, radius=9.1867, shapeName=dyn_robeBack_05_capsule, center=<7.1694 2.3664 0>, rotation=<rot 0 -90 0>}]
 dyn_robeSide_01_l [{shapeType=Capsule, length=15.9012, radius=9.1867, shapeName=dyn_robeBack_05_capsule, center=<7.1694 2.3664 0>, rotation=<rot 0 -90 0>}]
 dyn_robeSide_01_r [{shapeType=Capsule, length=15.9012, radius=9.1867, shapeName=dyn_robeBack_05_capsule, center=<7.1694 2.3664 0>, rotation=<rot 0 -90 0>}]
 dyn_robeSide_02_l [{shapeType=Capsule, length=15.9012, radius=9.1867, shapeName=dyn_robeBack_05_capsule, center=<7.1694 2.3664 0>, rotation=<rot 0 -90 0>}]
 dyn_robeSide_02_r [{shapeType=Capsule, length=15.9012, radius=9.1867, shapeName=dyn_robeBack_05_capsule, center=<7.1694 2.3664 0>, rotation=<rot 0 -90 0>}]
 dyn_robeSide_03_l [{shapeType=Capsule, length=15.9012, radius=9.1867, shapeName=dyn_robeBack_05_capsule, center=<7.1694 2.3664 0>, rotation=<rot 0 -90 0>}]
 dyn_robeSide_03_r [{shapeType=Capsule, length=15.9012, radius=9.1867, shapeName=dyn_robeBack_05_capsule, center=<7.1694 2.3664 0>, rotation=<rot 0 -90 0>}]
 dyn_robeSide_04_l [{shapeType=Capsule, length=15.9012, radius=9.1867, shapeName=dyn_robeBack_05_capsule, center=<7.1694 2.3664 0>, rotation=<rot 0 -90 0>}]
 dyn_robeSide_04_r [{shapeType=Capsule, length=15.9012, radius=9.1867, shapeName=dyn_robeBack_05_capsule, center=<7.1694 2.3664 0>, rotation=<rot 0 -90 0>}]
 dyn_robeSide_05_l [{shapeType=Capsule, length=15.9012, radius=9.1867, shapeName=dyn_robeBack_05_capsule, center=<7.1694 2.3664 0>, rotation=<rot 0 -90 0>}]
 dyn_robeSide_05_r [{shapeType=Capsule, length=15.9012, radius=9.1867, shapeName=dyn_robeBack_05_capsule, center=<7.1694 2.3664 0>, rotation=<rot 0 -90 0>}]
 foot_l []
 foot_r []
 head [{shapeType=Sphere, radius=17.3125, center=<11.028 -0.174 0>}]
 lowerarm_l []
 lowerarm_r []
 neck_02 [{shapeType=Capsule, length=13.4375, radius=7.3175, center=<-0.2642 -0.3408 0>, rotation=<rot 90 -171.1193 -175.3978>}]
 pelvis [{shapeType=Capsule, length=8.6456, radius=14.5388, center=<-2.4887 0.1881 0>}]
 spine_01 [{shapeType=Capsule, length=7.9438, radius=13.5586, center=<-1.7965 -6.2098 0>}]
 spine_02 [
  {shapeType=Capsule, length=7.3264, radius=13.6322, center=<-1.711 -10.6467 0.0001>}
  {shapeType=Capsule, length=29.3956, radius=11.5807, center=<2.5024 0.4992 3.9825>, rotation=<rot 90 0.0049 0.0049>}
  {shapeType=Capsule, length=29.875, radius=11.1013, center=<2.4704 0.4991 -3.9825>, rotation=<rot 90 -0.0396 -0.0396>}
 ]
 spine_03 [{shapeType=Capsule, length=8.2107, radius=13.7031, center=<-0.1791 -9.7241 0.0001>}]
 spine_04 [{shapeType=Capsule, length=9.4404, radius=14.1325, center=<2.766 -6.5622 0.0002>}]
 spine_05 [
  {shapeType=Capsule, length=8.0226, radius=15.695, center=<-2.6886 -2.8415 -0.0258>}
  {shapeType=Capsule, length=13.9518, radius=13.9425, center=<-0.2916 0.7706 0>}
 ]
 thigh_l []
 thigh_r []
 upperarm_l []
 upperarm_r []
}
constraints []
