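# Cloth collision physics asset for the DroveRay_Dip medium male player body, giving simulated cloth shapes to collide against.
# Capsules cover the pelvis and lower spine, with a tiny capsule at the root.
# A wide box on the ground_collision bone, its top at about that bone's origin, acts as a floor.
# Thigh, calf, foot and forearm bodies list no shapes, and no constraints link any bodies.
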PhysicsAsset /BRCosmetics/Characters/Player/Male/Medium/Bodies/M_MED_DroveRay_Dip/Meshes/M_MED_DroveRay_Dip_ClothCollision
bodies {
 calf_l []
 calf_r []
 foot_l []
 foot_r []
 ground_collision [{shapeType=Box, extentX=331.4972, extentY=332.5133, extentZ=73.6069, shapeName=root_box, center=<0 0 -36.7803>, rotation=<rot 0 -90 0>}]
 lowerarm_l []
 lowerarm_r []
 pelvis [
  {shapeType=Capsule, length=6.1467, radius=12.2272, center=<-2.4887 -0.6978 0>}
  {shapeType=Capsule, length=8, radius=13.9454, shapeName=pelvis_capsule, center=<-9.992 0.4037 0>, rotation=<rot 0 90 90>}
 ]
 root [{shapeType=Capsule, length=0.505, radius=0.505, shapeName=root_capsule, rotation=<rot 0 -90 0>}]
 spine_01 [{shapeType=Capsule, length=7.9663, radius=10.1267, center=<0.3095 -0.6696 0>}]
 spine_02 [{shapeType=Capsule, length=8.2792, radius=9.2496, center=<-1.0816 -0.436 -0.0002>}]
 thigh_l []
 thigh_r []
}
constraints []
